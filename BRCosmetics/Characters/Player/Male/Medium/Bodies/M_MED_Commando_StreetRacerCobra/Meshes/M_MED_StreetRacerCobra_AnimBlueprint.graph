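# Anim Blueprint for the StreetRacerCobra male commando body mesh, adding secondary physics motion to its dangling strings and zipper.
# It copies the pose from the attached parent mesh, then runs five chained Anim Dynamics bone solvers in component space.
# The four string bones blend at low alpha with spherical limits driven by upper spine bones, while the zipper simulates fully in actor space.
# The event graph contains no logic.
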AnimBlueprint /BRCosmetics/Characters/Player/Male/Medium/Bodies/M_MED_Commando_StreetRacerCobra/Meshes/M_MED_StreetRacerCobra_AnimBlueprint
target_skeleton=&/M_MED_StreetRacerCobra_Skeleton
graphs [
 AnimGraph {nodes [
   n1 = Anim Dynamics - Bone: dyn_string_r_01(ComponentPose=Local To Component(LocalPose=Copy Pose From Mesh() {bUseAttachedParent=true}), Alpha=0.25, ExternalForce="0,0,0") {
    angularBiasOverride=1
    angularDampingOverride=0.98
    bGravityOverrideInSimSpace=true
    bUseSphericalLimits=true
    lODThreshold=1
    numSolverIterationsPostUpdate=4
    numSolverIterationsPreUpdate=16
    alphaBoolBlend {blendOption=Linear}
    boundBone {boneName=dyn_string_r_01}
    physicsBodyDefinitions [{
      sphereCollisionRadius=0
      boundBone {boneName=dyn_string_r_01}
      boxExtents=<5 2 10>
      constraintSetup {angularTargetAxis=AxisZ, twistAxis=AxisZ, angularLimitsMax=<80 45 0>, angularLimitsMin=<0 -45 0>, angularTarget=<0 0 0>}
      localJointOffset=<0 1 -1>
     }]
    sphericalLimits [{limitRadius=13.5, drivingBone {boneName=spine_05}, sphereLocalOffset=<0 1 0>}]
   }
   n2 = Anim Dynamics - Bone: dyn_string_l_01(ComponentPose=@n1, Alpha=0.25, ExternalForce="0,0,0") {
    angularBiasOverride=1
    angularDampingOverride=0.98
    bGravityOverrideInSimSpace=true
    bUseSphericalLimits=true
    lODThreshold=1
    numSolverIterationsPostUpdate=4
    numSolverIterationsPreUpdate=16
    alphaBoolBlend {blendOption=Linear}
    boundBone {boneName=dyn_string_l_01}
    physicsBodyDefinitions [{
      sphereCollisionRadius=0
      boundBone {boneName=dyn_string_l_01}
      boxExtents=<5 2 10>
      constraintSetup {angularTargetAxis=AxisZ, twistAxis=AxisZ, angularLimitsMax=<80 45 0>, angularLimitsMin=<0 -45 0>, angularTarget=<0 0 0>}
      localJointOffset=<0 1 -1>
     }]
    sphericalLimits [{limitRadius=13.5, drivingBone {boneName=spine_05}, sphereLocalOffset=<0 1 0>}]
   }
   n3 = Anim Dynamics - Bone: dyn_string_r_02(ComponentPose=@n2, Alpha=0.3, ExternalForce="0,0,0") {
    angularBiasOverride=1
    angularDampingOverride=0.98
    bGravityOverrideInSimSpace=true
    bUseSphericalLimits=true
    lODThreshold=1
    numSolverIterationsPostUpdate=4
    numSolverIterationsPreUpdate=16
    alphaBoolBlend {blendOption=Linear}
    boundBone {boneName=dyn_string_r_02}
    physicsBodyDefinitions [{
      sphereCollisionRadius=0
      boundBone {boneName=dyn_string_r_02}
      boxExtents=<5 2 10>
      constraintSetup {angularTargetAxis=AxisZ, twistAxis=AxisZ, angularLimitsMax=<80 45 0>, angularLimitsMin=<0 -45 0>, angularTarget=<0 0 0>}
      localJointOffset=<0 1 -1>
     }]
    sphericalLimits [{limitRadius=13.5, drivingBone {boneName=spine_04}, sphereLocalOffset=<0 1.5 0>}]
   }
   n4 = Anim Dynamics - Bone: dyn_string_l_02(ComponentPose=@n3, Alpha=0.3, ExternalForce="0,0,0") {
    angularBiasOverride=1
    angularDampingOverride=0.98
    bGravityOverrideInSimSpace=true
    bUseSphericalLimits=true
    lODThreshold=1
    numSolverIterationsPostUpdate=4
    numSolverIterationsPreUpdate=16
    alphaBoolBlend {blendOption=Linear}
    boundBone {boneName=dyn_string_l_02}
    physicsBodyDefinitions [{
      sphereCollisionRadius=0
      boundBone {boneName=dyn_string_l_02}
      boxExtents=<5 2 10>
      constraintSetup {angularTargetAxis=AxisZ, twistAxis=AxisZ, angularLimitsMax=<80 45 0>, angularLimitsMin=<0 -45 0>, angularTarget=<0 0 0>}
      localJointOffset=<0 1 -1>
     }]
    sphericalLimits [{limitRadius=13.5, drivingBone {boneName=spine_04}, sphereLocalOffset=<0 1.5 0>}]
   }
   n5 = Anim Dynamics - Bone: dyn_zipper(ComponentPose=@n4, Alpha=1) {
    bGravityOverrideInSimSpace=true
    lODThreshold=1
    numSolverIterationsPostUpdate=4
    numSolverIterationsPreUpdate=16
    simulationSpace=Actor
    alphaBoolBlend {blendOption=Linear}
    boundBone {boneName=dyn_zipper}
    physicsBodyDefinitions [{
      sphereCollisionRadius=0
      boundBone {boneName=dyn_zipper}
      boxExtents=<5 5 5>
      constraintSetup {angularTargetAxis=AxisZ, twistAxis=AxisZ, angularLimitsMax=<5 30 0>, angularLimitsMin=<-90 -30 0>, angularTarget=<0 0 0>}
      localJointOffset=<0 0 -0.5>
     }]
   }
   Output Pose(Result=Component To Local(ComponentPose=@n5)) {layerGroup=None}
  ]}
 EventGraph {nodes [
   TryGetPawnOwner()
   Event BlueprintUpdateAnimation()
  ]}
]
variables []
